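# Editor utility widget front end for the AssetToBlender Python exporter, with FBX settings, export path, selection, result counts and temp-file controls.
# Each control runs `widget_api` Python strings; evaluated statements return text for the labels.
# Export forces FBX, re-sends placement and source, exports, then updates counts and calls guarded `RefreshUI`.
# Combo change handlers send literal '{SelectedItem}' instead of the chosen value.
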EditorUtilityWidgetBlueprint /Game/ArtTools/BlenderExporter/EUW_BlenderExporter
graphs [
 EventGraph {nodes [
   "On Clicked (ButtonBrowse)"() {is=title K2Node_ComponentBoundEvent}
   > n1 = "Execute Python Command (Advanced)"!(
    execute
    PythonCommand:String="__import__('AssetToBlender.ui.widget_api', fromlist=['']).browse_export_path()"
    ExecutionMode:EPythonCommandExecutionMode Enum=EvaluateStatement
    FileExecutionScope:EPythonFileExecutionScope Enum=Private
   ) -> (then, CommandResult:String, LogOutput:Array of Python Log Output Entry Structures, ReturnValue:bool)
   > Set Text(
    Text="To Text (String)"(InString=@n1)
    self:Text Box Object Reference=Get TextExportPath() -> (TextExportPath:Editor Utility Editable Text Box Object Reference)
   )
   "On Clicked (ButtonOpenFolder)"() {is=title K2Node_ComponentBoundEvent}
   > ExecutePythonCommand(PythonCommand="from AssetToBlender.ui.widget_api import open_export_folder; open_export_folder()")
   "On Clicked (ButtonCleanup)"() {is=title K2Node_ComponentBoundEvent}
   > ExecutePythonCommand(PythonCommand="from AssetToBlender.ui.widget_api import cleanup_temp_files; cleanup_temp_files()")
   "On Selection Changed (ComboFormat)"!() -> (OutputDelegate, then, SelectedItem:String, SelectionType:ESelectInfo Enum) {is=title K2Node_ComponentBoundEvent}
   > ExecutePythonCommand(PythonCommand="from AssetToBlender.ui.widget_api import set_export_format; set_export_format('{SelectedItem}')")
   "On Selection Changed (ComboPlacement)"!() -> (OutputDelegate, then, SelectedItem:String, SelectionType:ESelectInfo Enum) {is=title K2Node_ComponentBoundEvent}
   > ExecutePythonCommand(PythonCommand="from AssetToBlender.ui.widget_api import set_placement_mode; set_placement_mode('{SelectedItem}')")
   "On Selection Changed (ComboSource)"!() -> (OutputDelegate, then, SelectedItem:String, SelectionType:ESelectInfo Enum) {is=title K2Node_ComponentBoundEvent}
   > ExecutePythonCommand(PythonCommand="from AssetToBlender.ui.widget_api import set_source_mode; set_source_mode('{SelectedItem}')")
   "On Clicked (ButtonRefresh)"() {is=title K2Node_ComponentBoundEvent}
   > ExecutePythonCommand(PythonCommand="import AssetToBlender.ui.widget_api as api; api.refresh_ui()")
   > RefreshUI()
   > n2 = "Execute Python Command (Advanced)"(
    PythonCommand="__import__('AssetToBlender.ui.widget_api', fromlist=['']).get_selection_text()"
    ExecutionMode=EvaluateStatement
    FileExecutionScope=Private
   )
   > Set Text(
    Text=To Text!(Value:String=@n2) -> (ReturnValue:Text) {is=title K2Node_GenericToText}
    self:Text Object Reference=Get TextSelection() -> (TextSelection:Text Object Reference)
   )
   "On Clicked (ButtonExport)"() {is=title K2Node_ComponentBoundEvent}
   > ExecutePythonCommand(PythonCommand="import AssetToBlender.ui.widget_api as api; api.set_export_format('fbx')")
   n3 = MakeLiteralString(Value="from AssetToBlender.ui.widget_api import set_placement_mode; set_placement_mode('")
   n4 = GetSelectedOption!(self:"Combo Box (String) Object Reference"=Get ComboPlacement() -> (ComboPlacement:Editor Utility Combo Box String Object Reference)) -> (ReturnValue:String)
   > ExecutePythonCommand(PythonCommand=Append(A=Append(A=@n3, B=@n4), B=MakeLiteralString(Value="')")))
   n5 = MakeLiteralString(Value="from AssetToBlender.ui.widget_api import set_source_mode; set_source_mode('")
   n6 = GetSelectedOption(self=Get ComboSource() -> (ComboSource:Editor Utility Combo Box String Object Reference))
   > ExecutePythonCommand(PythonCommand=Append(A=Append(A=@n5, B=@n6), B=MakeLiteralString(Value="')")))
   > Clear and Invalidate Timer by Handle!(execute, WorldContextObject:Object Reference, Handle=Get RefreshTimerHandle() -> (RefreshTimerHandle:Timer Handle Structure)) -> (then)
   > ExecutePythonCommand(PythonCommand="from AssetToBlender.ui.widget_api import export_with_settings; print(export_with_settings())")
   > n7 = "Execute Python Command (Advanced)"(
    PythonCommand="__import__('AssetToBlender.ui.widget_api', fromlist=['']).get_success_count_text()"
    ExecutionMode=EvaluateStatement
    FileExecutionScope=Private
   )
   n8 = MakeLiteralString(Value=" Exported")
   > Set Text(Text="To Text (String)"(InString=Append(A=@n7, B=@n8)), self=Get TextSuccessCount() -> (TextSuccessCount:Text Object Reference))
   > n9 = "Execute Python Command (Advanced)"(
    PythonCommand="__import__('AssetToBlender.ui.widget_api', fromlist=['']).get_failed_count_text()"
    ExecutionMode=EvaluateStatement
    FileExecutionScope=Private
   )
   n10 = MakeLiteralString(Value=" Failed")
   > Set Text(Text="To Text (String)"(InString=Append(A=@n9, B=@n10)), self=Get TextFailedCount() -> (TextFailedCount:Text Object Reference))
   > RefreshUI()
   n11 = "On Check State Changed (CheckHighestQualityOnly)"!() -> (OutputDelegate, then, bIsChecked:bool) {is=title K2Node_ComponentBoundEvent}
   n12 = SelectString(A=True, B=False, bPickA=@n11.bIsChecked)
   > ExecutePythonCommand(PythonCommand=Append(A=Append(A="__import__('AssetToBlender.ui.widget_api', fromlist=['']).set_export_highest_quality_only(", B=@n12), B=")"))
   Event Construct()
   > Set Text(Text=MakeLiteralText(Value="LOCTEXT(\"C:\\\\Temp\\\\AssetToBlender\")"), self:Text Box Object Reference=Get TextExportPath())
   > SetSelectedOption!(
    execute
    self:"Combo Box (String) Object Reference"=Get ComboFormat() -> (ComboFormat:Editor Utility Combo Box String Object Reference)
    Option:String=FBX
   ) -> (then)
   > SetSelectedOption(self=Get ComboPlacement(), Option=Origin)
   > SetSelectedOption(self=Get ComboSource(), Option=Auto)
   > ExecutePythonCommand(PythonCommand="import AssetToBlender.ui.widget_api as api; api.refresh_ui()")
   > n13 = "Execute Python Command (Advanced)"(
    PythonCommand="__import__('AssetToBlender.ui.widget_api', fromlist=['']).get_export_highest_quality_only()"
    ExecutionMode=EvaluateStatement
    FileExecutionScope=Private
   )
   > SetIsChecked!(
    execute
    self:Check Box Object Reference=Get CheckHighestQualityOnly() -> (CheckHighestQualityOnly:Editor Utility Check Box Object Reference)
    InIsChecked:bool="Equal Exactly (String)"(A=@n13, B=True)
   ) -> (then)
   Event PreConstruct()
   Event Tick()
  ]}
 RefreshUI {nodes [
   RefreshUI() {is=title K2Node_FunctionEntry}
   > n1 = Branch(Condition=Get bIsRefreshing() -> (bIsRefreshing:bool))
   Set bIsRefreshing(execute=@n1.else, bIsRefreshing:bool=true) -> (Output_Get:bool)
   > n2 = "Execute Python Command (Advanced)"(
    PythonCommand="__import__('AssetToBlender.ui.widget_api', fromlist=['']).get_selection_text()"
    ExecutionMode=EvaluateStatement
    FileExecutionScope=Private
   )
   > Set Text(Text=To Text(Value=@n2), self:Text Object Reference=Get TextSelection())
   > n3 = "Execute Python Command (Advanced)"(
    PythonCommand="__import__('AssetToBlender.ui.widget_api', fromlist=['']).get_status_text()"
    ExecutionMode=EvaluateStatement
    FileExecutionScope=Private
   )
   > Set Text(Text=To Text(Value=@n3), self=Get TextTempStatus() -> (TextTempStatus:Text Object Reference))
   > Set bIsRefreshing(bIsRefreshing=false)
  ]}
]
variables [
 bIsRefreshing {type=boolean}
 ExportFormat {type=string, default=FBX}
 ExpotrtPath {type=string, default="C:\\Temp\\AssetToBlender"}
 LastFailedCount {type=integer}
 LastSuccessCount {type=integer}
 Placement {type=string, default=Origin}
 RefreshTimerHandle {type=TimerHandle}
 Source {type=string, default=Auto}
]
widgets {
 info {rootWidgetClass=&StackBox, widgetCount=52}
 widgets [
  StackBox_0 {widgetClassPath=&StackBox, widget=StackBox_0}
  HorizontalBox_0 {widgetClassPath=&HorizontalBox, parent=StackBox_0, widget=HorizontalBox_0, slot {class=StackBoxSlot}}
  VerticalBox_44 {
   widgetClassPath=&VerticalBox
   parent=HorizontalBox_0
   widget=VerticalBox_44
   slot {class=HorizontalBoxSlot, properties {padding=<ltrb 10 10 10 10>, size {sizeRule=Fill, value=0.6}}}
  }
  TextBlock_146 {
   widgetClassPath=&TextBlock
   parent=VerticalBox_44
   widget=TextBlock_146
   properties {
    text=Settings
    font {
     bForceMonospaced=false
     bMaterialIsStencil=false
     fontMaterial=null
     fontObject=&Roboto
     letterSpacing=0
     monospacedWidth=1
     size=12
     skewAmount=0
     typefaceFontName=Bold
     outlineSettings {
      bApplyOutlineToDropShadows=false
      bMiteredCorners=false
      bSeparateFillAlpha=false
      outlineMaterial=null
      outlineSize=0
      outlineColor=<rgba 0 0 0 1>
     }
    }
   }
   slot {class=VerticalBoxSlot, properties {padding=<ltrb 0 0 0 10>}}
  }
  Border_75 {
   widgetClassPath=&Border
   parent=VerticalBox_44
   widget=Border_75
   properties {brushColor=<rgba 0.15 0.15 0.15 1>}
   slot {class=VerticalBoxSlot, properties {padding=<ltrb 10 10 10 10>}}
  }
  rectangle {widgetClassPath=&VerticalBox, parent=Border_75, widget=rectangle, slot {class=BorderSlot}}
  HorizontalBox_42 {widgetClassPath=&HorizontalBox, parent=rectangle, widget=HorizontalBox_42, slot {class=VerticalBoxSlot}}
  TextBlock_190 {
   widgetClassPath=&TextBlock
   parent=HorizontalBox_42
   widget=TextBlock_190
   properties {
    text=Export Format
    font {
     bForceMonospaced=false
     bMaterialIsStencil=false
     fontMaterial=null
     fontObject=&Roboto
     letterSpacing=0
     monospacedWidth=1
     size=9
     skewAmount=0
     typefaceFontName=Bold
     outlineSettings {
      bApplyOutlineToDropShadows=false
      bMiteredCorners=false
      bSeparateFillAlpha=false
      outlineMaterial=null
      outlineSize=0
      outlineColor=<rgba 0 0 0 1>
     }
    }
   }
   slot {class=HorizontalBoxSlot, properties {horizontalAlignment=HAlign_Left, verticalAlignment=VAlign_Center, size {sizeRule=Fill, value=0.4}}}
  }
  ComboFormat {
   widgetClassPath=/Script/Blutility.EditorUtilityComboBoxString
   bIsVariable=true
   parent=HorizontalBox_42
   widget=ComboFormat
   properties {
    selectedOption=FBX
    defaultOptions [FBX]
    font {
     bForceMonospaced=false
     bMaterialIsStencil=false
     fontMaterial=null
     fontObject=&Roboto
     letterSpacing=0
     monospacedWidth=1
     size=9
     skewAmount=0
     typefaceFontName=Bold
     outlineSettings {
      bApplyOutlineToDropShadows=false
      bMiteredCorners=false
      bSeparateFillAlpha=false
      outlineMaterial=null
      outlineSize=0
      outlineColor=<rgba 0 0 0 1>
     }
    }
    foregroundColor {colorUseRule=UseColor_Specified, specifiedColor=<rgba 0.5271 0.5271 0.5271 1>}
   }
   slot {class=HorizontalBoxSlot, properties {size {sizeRule=Fill, value=0.6}}}
  }
  HorizontalBox_228 {widgetClassPath=&HorizontalBox, parent=rectangle, widget=HorizontalBox_228, slot {class=VerticalBoxSlot}}
  TextBlock_246 {
   widgetClassPath=&TextBlock
   parent=HorizontalBox_228
   widget=TextBlock_246
   properties {
    text=Placement
    font {
     bForceMonospaced=false
     bMaterialIsStencil=false
     fontMaterial=null
     fontObject=&Roboto
     letterSpacing=0
     monospacedWidth=1
     size=9
     skewAmount=0
     typefaceFontName=Bold
     outlineSettings {
      bApplyOutlineToDropShadows=false
      bMiteredCorners=false
      bSeparateFillAlpha=false
      outlineMaterial=null
      outlineSize=0
      outlineColor=<rgba 0 0 0 1>
     }
    }
   }
   slot {class=HorizontalBoxSlot, properties {horizontalAlignment=HAlign_Left, verticalAlignment=VAlign_Center, size {sizeRule=Fill, value=0.4}}}
  }
  ComboPlacement {
   widgetClassPath=&EditorUtilityComboBoxString
   bIsVariable=true
   parent=HorizontalBox_228
   widget=ComboPlacement
   properties {
    selectedOption=Origin
    defaultOptions [Origin, World, Origin Relative]
    font {
     bForceMonospaced=false
     bMaterialIsStencil=false
     fontMaterial=null
     fontObject=&Roboto
     letterSpacing=0
     monospacedWidth=1
     size=9
     skewAmount=0
     typefaceFontName=Bold
     outlineSettings {
      bApplyOutlineToDropShadows=false
      bMiteredCorners=false
      bSeparateFillAlpha=false
      outlineMaterial=null
      outlineSize=0
      outlineColor=<rgba 0 0 0 1>
     }
    }
    foregroundColor {colorUseRule=UseColor_Specified, specifiedColor=<rgba 0.5271 0.5271 0.5271 1>}
   }
   slot {class=HorizontalBoxSlot, properties {size {sizeRule=Fill, value=0.6}}}
  }
  HorizontalBox_301 {widgetClassPath=&HorizontalBox, parent=rectangle, widget=HorizontalBox_301, slot {class=VerticalBoxSlot}}
  TextBlock_311 {
   widgetClassPath=&TextBlock
   parent=HorizontalBox_301
   widget=TextBlock_311
   properties {
    text=Source
    font {
     bForceMonospaced=false
     bMaterialIsStencil=false
     fontMaterial=null
     fontObject=&Roboto
     letterSpacing=0
     monospacedWidth=1
     size=9
     skewAmount=0
     typefaceFontName=Bold
     outlineSettings {
      bApplyOutlineToDropShadows=false
      bMiteredCorners=false
      bSeparateFillAlpha=false
      outlineMaterial=null
      outlineSize=0
      outlineColor=<rgba 0 0 0 1>
     }
    }
   }
   slot {class=HorizontalBoxSlot, properties {horizontalAlignment=HAlign_Left, verticalAlignment=VAlign_Center, size {sizeRule=Fill, value=0.4}}}
  }
  ComboSource {
   widgetClassPath=&EditorUtilityComboBoxString
   bIsVariable=true
   parent=HorizontalBox_301
   widget=ComboSource
   properties {
    selectedOption=Auto
    defaultOptions [Auto, Level, Content Browser]
    font {
     bForceMonospaced=false
     bMaterialIsStencil=false
     fontMaterial=null
     fontObject=&Roboto
     letterSpacing=0
     monospacedWidth=1
     size=9
     skewAmount=0
     typefaceFontName=Bold
     outlineSettings {
      bApplyOutlineToDropShadows=false
      bMiteredCorners=false
      bSeparateFillAlpha=false
      outlineMaterial=null
      outlineSize=0
      outlineColor=<rgba 0 0 0 1>
     }
    }
    foregroundColor {colorUseRule=UseColor_Specified, specifiedColor=<rgba 0.5271 0.5271 0.5271 1>}
   }
   slot {class=HorizontalBoxSlot, properties {size {sizeRule=Fill, value=0.6}}}
  }
  HorizontalBox_354 {widgetClassPath=&HorizontalBox, parent=rectangle, widget=HorizontalBox_354, slot {class=VerticalBoxSlot}}
  TextBlock_1 {
   widgetClassPath=&TextBlock
   parent=HorizontalBox_354
   widget=TextBlock_1
   properties {
    text=Highest Quality Only
    font {
     bForceMonospaced=false
     bMaterialIsStencil=false
     fontMaterial=null
     fontObject=&Roboto
     letterSpacing=0
     monospacedWidth=1
     size=9
     skewAmount=0
     typefaceFontName=Bold
     outlineSettings {
      bApplyOutlineToDropShadows=false
      bMiteredCorners=false
      bSeparateFillAlpha=false
      outlineMaterial=null
      outlineSize=0
      outlineColor=<rgba 0 0 0 1>
     }
    }
   }
   slot {
    class=HorizontalBoxSlot
    properties {horizontalAlignment=HAlign_Left, verticalAlignment=VAlign_Center, padding=<ltrb 0.4 0.4 0.4 0.4>, size {sizeRule=Fill, value=1}}
   }
  }
  CheckHighestQualityOnly {
   widgetClassPath=&EditorUtilityCheckBox
   bIsVariable=true
   parent=HorizontalBox_354
   widget=CheckHighestQualityOnly
   slot {class=HorizontalBoxSlot, properties {horizontalAlignment=HAlign_Left, size {sizeRule=Fill, value=0.6}}}
  }
  Spacer_57 {widgetClassPath=&Spacer, parent=VerticalBox_44, widget=Spacer_57, properties {size=<1 15>}, slot {class=VerticalBoxSlot}}
  TextBlock_432 {
   widgetClassPath=&TextBlock
   parent=VerticalBox_44
   widget=TextBlock_432
   properties {
    text=Export Path
    font {
     bForceMonospaced=false
     bMaterialIsStencil=false
     fontMaterial=null
     fontObject=&Roboto
     letterSpacing=0
     monospacedWidth=1
     size=12
     skewAmount=0
     typefaceFontName=Bold
     outlineSettings {
      bApplyOutlineToDropShadows=false
      bMiteredCorners=false
      bSeparateFillAlpha=false
      outlineMaterial=null
      outlineSize=0
      outlineColor=<rgba 0 0 0 1>
     }
    }
   }
   slot {class=VerticalBoxSlot, properties {padding=<ltrb 0 0 0 10>}}
  }
  HorizontalBox_1551 {widgetClassPath=&HorizontalBox, parent=VerticalBox_44, widget=HorizontalBox_1551, slot {class=VerticalBoxSlot}}
  TextExportPath {
   widgetClassPath=&EditorUtilityEditableTextBox
   bIsVariable=true
   parent=HorizontalBox_1551
   widget=TextExportPath
   properties {text="C:\\Temp\\AssetToBlender"}
   slot {class=HorizontalBoxSlot, properties {size {sizeRule=Fill, value=1}}}
  }
  ButtonBrowse {widgetClassPath=&EditorUtilityButton, parent=HorizontalBox_1551, widget=ButtonBrowse, slot {class=HorizontalBoxSlot}}
  TextBlock_524 {
   widgetClassPath=&TextBlock
   parent=ButtonBrowse
   widget=TextBlock_524
   properties {
    text=Browse...
    font {
     bForceMonospaced=false
     bMaterialIsStencil=false
     fontMaterial=null
     fontObject=&Roboto
     letterSpacing=0
     monospacedWidth=1
     size=9
     skewAmount=0
     typefaceFontName=Bold
     outlineSettings {
      bApplyOutlineToDropShadows=false
      bMiteredCorners=false
      bSeparateFillAlpha=false
      outlineMaterial=null
      outlineSize=0
      outlineColor=<rgba 0 0 0 1>
     }
    }
   }
   slot {class=ButtonSlot}
  }
  VerticalBox_183 {
   widgetClassPath=&VerticalBox
   parent=HorizontalBox_0
   widget=VerticalBox_183
   slot {class=HorizontalBoxSlot, properties {padding=<ltrb 10 10 10 10>, size {sizeRule=Fill, value=0.55}}}
  }
  TextBlock_563 {
   widgetClassPath=&TextBlock
   parent=VerticalBox_183
   widget=TextBlock_563
   properties {
    text=Export
    font {
     bForceMonospaced=false
     bMaterialIsStencil=false
     fontMaterial=null
     fontObject=&Roboto
     letterSpacing=0
     monospacedWidth=1
     size=12
     skewAmount=0
     typefaceFontName=Bold
     outlineSettings {
      bApplyOutlineToDropShadows=false
      bMiteredCorners=false
      bSeparateFillAlpha=false
      outlineMaterial=null
      outlineSize=0
      outlineColor=<rgba 0 0 0 1>
     }
    }
   }
   slot {class=VerticalBoxSlot, properties {padding=<ltrb 0 0 0 10>}}
  }
  ButtonExport {
   widgetClassPath=&EditorUtilityButton
   parent=VerticalBox_183
   widget=ButtonExport
   properties {widgetStyle {
     clickedSlateSound {resourceObject=null}
     disabled {
      bIsDynamicallyLoaded=false
      drawAs=RoundedBox
      imageType=NoImage
      mirroring=NoMirror
      resourceName=None
      resourceObject=null
      tiling=NoTile
      imageSize=<0 0>
      margin=<ltrb 0 0 0 0>
      outlineSettings {
       bUseBrushTransparency=false
       roundingType=FixedRadius
       width=1
       color {colorUseRule=UseColor_ColorTable, specifiedColor=<rgba 0 0 0 0>}
       cornerRadii=<4 4 4 4>
      }
      tintColor {colorUseRule=UseColor_Specified, specifiedColor=<rgba 0.0395 0.0395 0.0395 1>}
      uVRegion {bIsValid=false, max=<0 0>, min=<0 0>}
     }
     disabledForeground {colorUseRule=UseColor_ColorTable, specifiedColor=<rgba 0 0 0 0>}
     hovered {
      bIsDynamicallyLoaded=false
      drawAs=RoundedBox
      imageType=NoImage
      mirroring=NoMirror
      resourceName=None
      resourceObject=null
      tiling=NoTile
      imageSize=<32 32>
      margin=<ltrb 0 0 0 0>
      outlineSettings {
       bUseBrushTransparency=false
       roundingType=FixedRadius
       width=1
       color {colorUseRule=UseColor_ColorTable, specifiedColor=<rgba 0 0 0 0>}
       cornerRadii=<4 4 4 4>
      }
      tintColor {colorUseRule=UseColor_Specified, specifiedColor=<rgba 0.1 0.7 0.3 1>}
      uVRegion {bIsValid=false, max=<0 0>, min=<0 0>}
     }
     hoveredForeground {colorUseRule=UseColor_ColorTable, specifiedColor=<rgba 0 0 0 0>}
     hoveredSlateSound {resourceObject=null}
     normal {
      bIsDynamicallyLoaded=false
      drawAs=RoundedBox
      imageType=NoImage
      mirroring=NoMirror
      resourceName=None
      resourceObject=null
      tiling=NoTile
      imageSize=<32 32>
      margin=<ltrb 0 0 0 0>
      outlineSettings {
       bUseBrushTransparency=false
       roundingType=FixedRadius
       width=1
       color {colorUseRule=UseColor_ColorTable, specifiedColor=<rgba 0 0 0 0>}
       cornerRadii=<4 4 4 4>
      }
      tintColor {colorUseRule=UseColor_Specified, specifiedColor=<rgba 0.0223 0.1562 0.067 1>}
      uVRegion {bIsValid=false, max=<0 0>, min=<0 0>}
     }
     normalForeground {colorUseRule=UseColor_ColorTable, specifiedColor=<rgba 0 0 0 0>}
     normalPadding=<ltrb 12 1.5 12 1.5>
     pressed {
      bIsDynamicallyLoaded=false
      drawAs=RoundedBox
      imageType=NoImage
      mirroring=NoMirror
      resourceName=None
      resourceObject=null
      tiling=NoTile
      imageSize=<32 32>
      margin=<ltrb 0 0 0 0>
      outlineSettings {
       bUseBrushTransparency=false
       roundingType=FixedRadius
       width=1
       color {colorUseRule=UseColor_ColorTable, specifiedColor=<rgba 0 0 0 0>}
       cornerRadii=<4 4 4 4>
      }
      tintColor {colorUseRule=UseColor_Specified, specifiedColor=<rgba 0.0521 0.0521 0.0521 1>}
      uVRegion {bIsValid=false, max=<0 0>, min=<0 0>}
     }
     pressedForeground {colorUseRule=UseColor_ColorTable, specifiedColor=<rgba 0 0 0 0>}
     pressedPadding=<ltrb 12 2.5 12 0.5>
     pressedSlateSound {resourceObject=null}
    }}
   slot {class=VerticalBoxSlot, properties {verticalAlignment=VAlign_Center}}
  }
  TextBlock_565 {
   widgetClassPath=&TextBlock
   parent=ButtonExport
   widget=TextBlock_565
   properties {
    text=Export
    font {
     bForceMonospaced=false
     bMaterialIsStencil=false
     fontMaterial=null
     fontObject=&Roboto
     letterSpacing=0
     monospacedWidth=1
     size=13.5
     skewAmount=0
     typefaceFontName=Bold
     outlineSettings {
      bApplyOutlineToDropShadows=false
      bMiteredCorners=false
      bSeparateFillAlpha=false
      outlineMaterial=null
      outlineSize=0
      outlineColor=<rgba 0 0 0 1>
     }
    }
   }
   slot {class=ButtonSlot}
  }
  Spacer_154 {widgetClassPath=&Spacer, parent=VerticalBox_183, widget=Spacer_154, properties {size=<1 15>}, slot {class=VerticalBoxSlot}}
  SelectionHeader {
   widgetClassPath=&TextBlock
   parent=VerticalBox_183
   widget=SelectionHeader
   properties {
    text=Selection
    font {
     bForceMonospaced=false
     bMaterialIsStencil=false
     fontMaterial=null
     fontObject=&Roboto
     letterSpacing=0
     monospacedWidth=1
     size=12
     skewAmount=0
     typefaceFontName=Bold
     outlineSettings {
      bApplyOutlineToDropShadows=false
      bMiteredCorners=false
      bSeparateFillAlpha=false
      outlineMaterial=null
      outlineSize=0
      outlineColor=<rgba 0 0 0 1>
     }
    }
   }
   slot {class=VerticalBoxSlot, properties {padding=<ltrb 0 5 0 0>}}
  }
  TextSelection {
   widgetClassPath=&TextBlock
   bIsVariable=true
   parent=VerticalBox_183
   widget=TextSelection
   properties {
    text=No Selection
    colorAndOpacity {colorUseRule=UseColor_Specified, specifiedColor=<rgba 0.3021 0.3021 0.3021 1>}
    font {
     bForceMonospaced=false
     bMaterialIsStencil=false
     fontMaterial=null
     fontObject=&Roboto
     letterSpacing=0
     monospacedWidth=1
     size=10.5
     skewAmount=0
     typefaceFontName=Regular
     outlineSettings {
      bApplyOutlineToDropShadows=false
      bMiteredCorners=false
      bSeparateFillAlpha=false
      outlineMaterial=null
      outlineSize=0
      outlineColor=<rgba 0 0 0 1>
     }
    }
   }
   slot {class=VerticalBoxSlot, properties {padding=<ltrb 0 5 0 0>}}
  }
  Spacer_300 {widgetClassPath=&Spacer, parent=VerticalBox_183, widget=Spacer_300, properties {size=<1 10>}, slot {class=VerticalBoxSlot}}
  TextBlock_1100 {
   widgetClassPath=&TextBlock
   parent=VerticalBox_183
   widget=TextBlock_1100
   properties {
    text=Export Results
    font {
     bForceMonospaced=false
     bMaterialIsStencil=false
     fontMaterial=null
     fontObject=&Roboto
     letterSpacing=0
     monospacedWidth=1
     size=12
     skewAmount=0
     typefaceFontName=Bold
     outlineSettings {
      bApplyOutlineToDropShadows=false
      bMiteredCorners=false
      bSeparateFillAlpha=false
      outlineMaterial=null
      outlineSize=0
      outlineColor=<rgba 0 0 0 1>
     }
    }
   }
   slot {class=VerticalBoxSlot, properties {padding=<ltrb 0 0 0 10>}}
  }
  HorizontalBox_1738 {widgetClassPath=&HorizontalBox, parent=VerticalBox_183, widget=HorizontalBox_1738, slot {class=VerticalBoxSlot}}
  TextBlock_1177 {
   widgetClassPath=&TextBlock
   parent=HorizontalBox_1738
   widget=TextBlock_1177
   properties {
    text=Last Export:
    colorAndOpacity {colorUseRule=UseColor_Specified, specifiedColor=<rgba 0.3021 0.3021 0.3021 1>}
    font {
     bForceMonospaced=false
     bMaterialIsStencil=false
     fontMaterial=null
     fontObject=&Roboto
     letterSpacing=0
     monospacedWidth=1
     size=10.5
     skewAmount=0
     typefaceFontName=Regular
     outlineSettings {
      bApplyOutlineToDropShadows=false
      bMiteredCorners=false
      bSeparateFillAlpha=false
      outlineMaterial=null
      outlineSize=0
      outlineColor=<rgba 0 0 0 1>
     }
    }
   }
   slot {class=HorizontalBoxSlot}
  }
  TextSuccessCount {
   widgetClassPath=&TextBlock
   bIsVariable=true
   parent=HorizontalBox_1738
   widget=TextSuccessCount
   properties {
    text=0 succeeded
    colorAndOpacity {colorUseRule=UseColor_Specified, specifiedColor=<rgba 0.2 0.8 0.2 1>}
    font {
     bForceMonospaced=false
     bMaterialIsStencil=false
     fontMaterial=null
     fontObject=&Roboto
     letterSpacing=0
     monospacedWidth=1
     size=10.5
     skewAmount=0
     typefaceFontName=Bold
     outlineSettings {
      bApplyOutlineToDropShadows=false
      bMiteredCorners=false
      bSeparateFillAlpha=false
      outlineMaterial=null
      outlineSize=0
      outlineColor=<rgba 0 0 0 1>
     }
    }
   }
   slot {class=HorizontalBoxSlot}
  }
  TextBlock_1544 {
   widgetClassPath=&TextBlock
   parent=HorizontalBox_1738
   widget=TextBlock_1544
   properties {
    text=" • "
    font {
     bForceMonospaced=false
     bMaterialIsStencil=false
     fontMaterial=null
     fontObject=&Roboto
     letterSpacing=0
     monospacedWidth=1
     size=10.5
     skewAmount=0
     typefaceFontName=Bold
     outlineSettings {
      bApplyOutlineToDropShadows=false
      bMiteredCorners=false
      bSeparateFillAlpha=false
      outlineMaterial=null
      outlineSize=0
      outlineColor=<rgba 0 0 0 1>
     }
    }
   }
   slot {class=HorizontalBoxSlot}
  }
  TextFailedCount {
   widgetClassPath=&TextBlock
   bIsVariable=true
   parent=HorizontalBox_1738
   widget=TextFailedCount
   properties {
    text=0 failed
    colorAndOpacity {colorUseRule=UseColor_Specified, specifiedColor=<rgba 0.9 0.2 0.2 1>}
    font {
     bForceMonospaced=false
     bMaterialIsStencil=false
     fontMaterial=null
     fontObject=&Roboto
     letterSpacing=0
     monospacedWidth=1
     size=10.5
     skewAmount=0
     typefaceFontName=Bold
     outlineSettings {
      bApplyOutlineToDropShadows=false
      bMiteredCorners=false
      bSeparateFillAlpha=false
      outlineMaterial=null
      outlineSize=0
      outlineColor=<rgba 0 0 0 1>
     }
    }
   }
   slot {class=HorizontalBoxSlot}
  }
  Spacer_426 {widgetClassPath=&Spacer, parent=VerticalBox_183, widget=Spacer_426, properties {size=<1 10>}, slot {class=VerticalBoxSlot}}
  TextBlock_2153 {
   widgetClassPath=&TextBlock
   parent=VerticalBox_183
   widget=TextBlock_2153
   properties {
    text=Temp Files
    font {
     bForceMonospaced=false
     bMaterialIsStencil=false
     fontMaterial=null
     fontObject=&Roboto
     letterSpacing=0
     monospacedWidth=1
     size=12
     skewAmount=0
     typefaceFontName=Bold
     outlineSettings {
      bApplyOutlineToDropShadows=false
      bMiteredCorners=false
      bSeparateFillAlpha=false
      outlineMaterial=null
      outlineSize=0
      outlineColor=<rgba 0 0 0 1>
     }
    }
   }
   slot {class=VerticalBoxSlot, properties {padding=<ltrb 0 0 0 10>}}
  }
  TextTempStatus {
   widgetClassPath=&TextBlock
   bIsVariable=true
   parent=VerticalBox_183
   widget=TextTempStatus
   properties {
    text="0 files, 0 MB"
    colorAndOpacity {colorUseRule=UseColor_Specified, specifiedColor=<rgba 0.3021 0.3021 0.3021 1>}
    font {
     bForceMonospaced=false
     bMaterialIsStencil=false
     fontMaterial=null
     fontObject=&Roboto
     letterSpacing=0
     monospacedWidth=1
     size=10.5
     skewAmount=0
     typefaceFontName=Regular
     outlineSettings {
      bApplyOutlineToDropShadows=false
      bMiteredCorners=false
      bSeparateFillAlpha=false
      outlineMaterial=null
      outlineSize=0
      outlineColor=<rgba 0 0 0 1>
     }
    }
   }
   slot {class=VerticalBoxSlot, properties {padding=<ltrb 0 5 0 0>}}
  }
  HorizontalBox_1845 {widgetClassPath=&HorizontalBox, parent=VerticalBox_183, widget=HorizontalBox_1845, slot {class=VerticalBoxSlot}}
  ButtonOpenFolder {widgetClassPath=&EditorUtilityButton, parent=HorizontalBox_1845, widget=ButtonOpenFolder, slot {class=HorizontalBoxSlot}}
  TextBlock_2378 {
   widgetClassPath=&TextBlock
   parent=ButtonOpenFolder
   widget=TextBlock_2378
   properties {
    text=Open Folder
    font {
     bForceMonospaced=false
     bMaterialIsStencil=false
     fontMaterial=null
     fontObject=&Roboto
     letterSpacing=0
     monospacedWidth=1
     size=9
     skewAmount=0
     typefaceFontName=Bold
     outlineSettings {
      bApplyOutlineToDropShadows=false
      bMiteredCorners=false
      bSeparateFillAlpha=false
      outlineMaterial=null
      outlineSize=0
      outlineColor=<rgba 0 0 0 1>
     }
    }
   }
   slot {class=ButtonSlot}
  }
  Spacer_923 {widgetClassPath=&Spacer, parent=HorizontalBox_1845, widget=Spacer_923, properties {size=<10 1>}, slot {class=HorizontalBoxSlot}}
  ButtonCleanup {
   widgetClassPath=&EditorUtilityButton
   parent=HorizontalBox_1845
   widget=ButtonCleanup
   properties {widgetStyle {
     clickedSlateSound {resourceObject=null}
     disabled {
      bIsDynamicallyLoaded=false
      drawAs=RoundedBox
      imageType=NoImage
      mirroring=NoMirror
      resourceName=None
      resourceObject=null
      tiling=NoTile
      imageSize=<0 0>
      margin=<ltrb 0 0 0 0>
      outlineSettings {
       bUseBrushTransparency=false
       roundingType=FixedRadius
       width=1
       color {colorUseRule=UseColor_ColorTable, specifiedColor=<rgba 0 0 0 0>}
       cornerRadii=<4 4 4 4>
      }
      tintColor {colorUseRule=UseColor_Specified, specifiedColor=<rgba 0.0395 0.0395 0.0395 1>}
      uVRegion {bIsValid=false, max=<0 0>, min=<0 0>}
     }
     disabledForeground {colorUseRule=UseColor_ColorTable, specifiedColor=<rgba 0 0 0 0>}
     hovered {
      bIsDynamicallyLoaded=false
      drawAs=RoundedBox
      imageType=NoImage
      mirroring=NoMirror
      resourceName=None
      resourceObject=null
      tiling=NoTile
      imageSize=<32 32>
      margin=<ltrb 0 0 0 0>
      outlineSettings {
       bUseBrushTransparency=false
       roundingType=FixedRadius
       width=1
       color {colorUseRule=UseColor_ColorTable, specifiedColor=<rgba 0 0 0 0>}
       cornerRadii=<4 4 4 4>
      }
      tintColor {colorUseRule=UseColor_Specified, specifiedColor=<rgba 0.0953 0.0953 0.0953 1>}
      uVRegion {bIsValid=false, max=<0 0>, min=<0 0>}
     }
     hoveredForeground {colorUseRule=UseColor_ColorTable, specifiedColor=<rgba 0 0 0 0>}
     hoveredSlateSound {resourceObject=null}
     normal {
      bIsDynamicallyLoaded=false
      drawAs=RoundedBox
      imageType=NoImage
      mirroring=NoMirror
      resourceName=None
      resourceObject=null
      tiling=NoTile
      imageSize=<32 32>
      margin=<ltrb 0 0 0 0>
      outlineSettings {
       bUseBrushTransparency=false
       roundingType=FixedRadius
       width=1
       color {colorUseRule=UseColor_ColorTable, specifiedColor=<rgba 0 0 0 0>}
       cornerRadii=<4 4 4 4>
      }
      tintColor {colorUseRule=UseColor_Specified, specifiedColor=<rgba 0.6 0.15 0.15 1>}
      uVRegion {bIsValid=false, max=<0 0>, min=<0 0>}
     }
     normalForeground {colorUseRule=UseColor_ColorTable, specifiedColor=<rgba 0 0 0 0>}
     normalPadding=<ltrb 12 1.5 12 1.5>
     pressed {
      bIsDynamicallyLoaded=false
      drawAs=RoundedBox
      imageType=NoImage
      mirroring=NoMirror
      resourceName=None
      resourceObject=null
      tiling=NoTile
      imageSize=<32 32>
      margin=<ltrb 0 0 0 0>
      outlineSettings {
       bUseBrushTransparency=false
       roundingType=FixedRadius
       width=1
       color {colorUseRule=UseColor_ColorTable, specifiedColor=<rgba 0 0 0 0>}
       cornerRadii=<4 4 4 4>
      }
      tintColor {colorUseRule=UseColor_Specified, specifiedColor=<rgba 0.0284 0.0284 0.0284 1>}
      uVRegion {bIsValid=false, max=<0 0>, min=<0 0>}
     }
     pressedForeground {colorUseRule=UseColor_ColorTable, specifiedColor=<rgba 0 0 0 0>}
     pressedPadding=<ltrb 12 2.5 12 0.5>
     pressedSlateSound {resourceObject=null}
    }}
   slot {class=HorizontalBoxSlot}
  }
  TextBlock_2484 {
   widgetClassPath=&TextBlock
   parent=ButtonCleanup
   widget=TextBlock_2484
   properties {
    text=Cleanup
    font {
     bForceMonospaced=false
     bMaterialIsStencil=false
     fontMaterial=null
     fontObject=&Roboto
     letterSpacing=0
     monospacedWidth=1
     size=9
     skewAmount=0
     typefaceFontName=Bold
     outlineSettings {
      bApplyOutlineToDropShadows=false
      bMiteredCorners=false
      bSeparateFillAlpha=false
      outlineMaterial=null
      outlineSize=0
      outlineColor=<rgba 0 0 0 1>
     }
    }
   }
   slot {class=ButtonSlot}
  }
  ButtonRefresh {widgetClassPath=&EditorUtilityButton, parent=VerticalBox_183, widget=ButtonRefresh, slot {class=VerticalBoxSlot}}
  TextBlock_2486 {widgetClassPath=&TextBlock, parent=ButtonRefresh, widget=TextBlock_2486, properties {text="↻"}, slot {class=ButtonSlot}}
  HorizontalBox_2247 {widgetClassPath=&HorizontalBox, parent=HorizontalBox_0, widget=HorizontalBox_2247, slot {class=HorizontalBoxSlot}}
  HorizontalBox_2124 {widgetClassPath=&HorizontalBox, parent=StackBox_0, widget=HorizontalBox_2124, slot {class=StackBoxSlot}}
  HorizontalBox_2184 {widgetClassPath=&HorizontalBox, parent=StackBox_0, widget=HorizontalBox_2184, slot {class=StackBoxSlot}}
 ]
}
